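# Sound Cue for small electric spark one-shots, filed with the Athena EnergyCloud quest sounds.
# A Random node picks among eight equally weighted spark wave players at half volume in the Athena ambient props sound class.
# Inline attenuation overrides give it a short-range falloff with listener focus and low-pass occlusion.
# A concurrency override caps simultaneous instances at five, with a short release on stolen voices.
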SoundCue /Game/Athena/Sounds/Quests/EnergyCloud/SC_Electric_Sparks_Small
root=Random_1
nodes [
 {
  id=Random_1
  children [WavePlayer_0, WavePlayer_4, WavePlayer_5, WavePlayer_6, WavePlayer_7, WavePlayer_8, WavePlayer_9, WavePlayer_10]
  properties {weights [1,1,1,1,1,1,1,1]}
 }
 {id=WavePlayer_0, properties {soundWaveAssetPtr=&/Electric_Spark_01}}
 {id=WavePlayer_10, properties {soundWaveAssetPtr=&/Electric_Spark_08}}
 {id=WavePlayer_4, properties {soundWaveAssetPtr=&/Electric_Spark_02}}
 {id=WavePlayer_5, properties {soundWaveAssetPtr=&/Electric_Spark_03}}
 {id=WavePlayer_6, properties {soundWaveAssetPtr=&/Electric_Spark_04}}
 {id=WavePlayer_7, properties {soundWaveAssetPtr=&/Electric_Spark_05}}
 {id=WavePlayer_8, properties {soundWaveAssetPtr=&/Electric_Spark_06}}
 {id=WavePlayer_9, properties {soundWaveAssetPtr=&/Electric_Spark_07}}
]
properties {
 bOverrideAttenuation=true
 bOverrideConcurrency=true
 maxDistance=700
 soundClassObject=&Fort_Ambient_Props_Athena
 volumeMultiplier=0.5
 attenuationOverrides {
  bEnableListenerFocus=true
  bEnableOcclusion=true
  falloffDistance=500
  focusVolumeAttenuation=2
  occlusionLowPassFilterFrequency=1000
  attenuationShapeExtents {x=200}
 }
 concurrencyOverrides {maxCount=5, voiceStealReleaseTime=0.3, volumeScaleAttackTime=0.3}
}
