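# Sound cue for a small electric spark arc spot effect, stored with the Energy Cloud quest sounds.
# A random node picks one of six equally weighted spark waves and feeds it through a close electrical spark attenuation node.
# A root modulator adds slight pitch and volume variation to each play.
# The cue plays at reduced volume in the in-game SFX class, with a concurrency override capping simultaneous voices.
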SoundCue /Game/Athena/Sounds/Quests/EnergyCloud/Spot_Electric_Spark_Arc_Small_Cue
root=Modulator_0
nodes [
 {
  id=Attenuation_0
  children [Random_0]
  properties {attenuationSettings=&/Fort_Electrical_Spark_Close_Air_Att}
 }
 {id=Modulator_0, children [Attenuation_0], properties {pitchMax=1.2, pitchMin=0.9, volumeMax=1.1}}
 {id=Random_0, children [WavePlayer_0,WavePlayer_1,WavePlayer_2,WavePlayer_3,WavePlayer_4,WavePlayer_5], properties {weights [1,1,1,1,1,1]}}
 {id=WavePlayer_0, properties {soundWaveAssetPtr=&/Spot_Electric_Spark_Arc_Small_06}}
 {id=WavePlayer_1, properties {soundWaveAssetPtr=&/Spot_Electric_Spark_Arc_Small_01}}
 {id=WavePlayer_2, properties {soundWaveAssetPtr=&/Spot_Electric_Spark_Arc_Small_02}}
 {id=WavePlayer_3, properties {soundWaveAssetPtr=&/Spot_Electric_Spark_Arc_Small_03}}
 {id=WavePlayer_4, properties {soundWaveAssetPtr=&/Spot_Electric_Spark_Arc_Small_04}}
 {id=WavePlayer_5, properties {soundWaveAssetPtr=&/Spot_Electric_Spark_Arc_Small_05}}
]
properties {
 bOverrideConcurrency=true
 maxDistance=3500
 soundClassObject=&Fort_SFX_InGame
 volumeMultiplier=0.35
 concurrencyOverrides {maxCount=3, voiceStealReleaseTime=0.25}
}
