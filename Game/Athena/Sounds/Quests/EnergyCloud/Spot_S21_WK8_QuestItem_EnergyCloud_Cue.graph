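# Looping spatialized ambient sound cue for the Season 21 Week 8 Energy Cloud quest item, in the quest sound class.
# A single looping wave plays through a boosted mixer that feeds every quality level identically.
# Two nested, endlessly looping envelopers of different lengths add slow drift: the inner one gently dips pitch, the outer one gently dips volume.
# It references a supply-drop hidden-star ambient attenuation asset.
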SoundCue /Game/Athena/Sounds/Quests/EnergyCloud/Spot_S21_WK8_QuestItem_EnergyCloud_Cue
root=Enveloper_2
nodes [
 {
  id=Enveloper_1
  children [QualityLevel_0]
  properties {
   bLoop=true
   bLoopIndefinitely=true
   loopEnd=30
   pitchCurve {editorCurveData {keys [{interpMode=RCIM_Cubic, time=0}, <at 15.9468 0.98 RCIM_Cubic>, {interpMode=RCIM_Cubic, time=30}]}}
   volumeCurve {editorCurveData {keys [<at 0 1>]}}
  }
 }
 {
  id=Enveloper_2
  children [Enveloper_1]
  properties {
   bLoop=true
   bLoopIndefinitely=true
   loopEnd=20
   pitchCurve {editorCurveData {keys [{time=0}]}}
   volumeCurve {editorCurveData {keys [
      {arriveTangent=-0.01, interpMode=RCIM_Cubic, leaveTangent=-0.01, time=0, value=1}
      <at 10 0.9 RCIM_Cubic>
      {arriveTangent=0.01, interpMode=RCIM_Cubic, leaveTangent=0.01, time=20, value=1}
     ]}}
  }
 }
 {id=Mixer_0, children [WavePlayer_2], properties {inputVolume [2.5]}}
 {id=QualityLevel_0, children [Mixer_0,Mixer_0,Mixer_0,Mixer_0]}
 {id=WavePlayer_2, properties {bLooping=true, soundWaveAssetPtr=&/Spot_QuestItem_Energy_Cloud_Loop}}
]
properties {
 attenuationSettings=/Game/Athena/Sounds/SupplyDrops/Fort_HiddenStar_Ambient_Loop_Att
 maxDistance=2500
 soundClassObject=&Fort_Quests
 volumeMultiplier=1
 attenuationOverrides {falloffDistance=700, attenuationShapeExtents {x=100}}
}
